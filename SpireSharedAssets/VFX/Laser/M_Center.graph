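# Translucent, unlit, two-sided Niagara sprite and mesh particle material for the center of a laser effect, with swirling radial smoke.
# Noise textures pan in radial coordinates, timed by a dynamic parameter, with the first noise distorting both smoke lookups.
# The noise modulates particle color for emissive.
# Opacity combines particle alpha and smoke, shaped by soft circular masks, an optional center cutout and view-angle fade.
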Material /SpireSharedAssets/VFX/Laser/M_Center
graph {
 properties {
  bAutomaticallySetUsageInEditor=false
  bUsedWithNiagaraMeshParticles=true
  bUsedWithNiagaraSprites=true
  blendMode=BLEND_Translucent
  shadingModel=MSM_Unlit
  twoSided=true
  usedShadingModels=Unlit
 }
 nodes [
  n1 = ParticleColor()
  n2 = ScalarParameter() {parameterName=ColorNoiseLerpMin}
  n3 = MaterialFunctionCall() {materialFunction=&VectorToRadialValue}
  n4 = Constant2Vector() {g=0.5, r=3}
  n5 = DynamicParameter() {paramNames [Time,Param2,Param3,Param4]}
  n6 = TextureSample(UVs=Panner(Coordinate=Multiply(A=@n3.Radial Coordinates, B=@n4), Time=@n5.Time) {speedY=-0.2}) {texture=/Game/Effects/Fort_Effects/Textures/Tile/T_Noise_03}
  n7 = Multiply(
   A=MaterialFunctionCall() {materialFunction=&VectorToRadialValue}.Radial Coordinates
   B=Constant2Vector() {g=0.3, r=2}
  )
  n8 = Multiply(A=@n6.R) {constB=0.2}
  n9 = TextureSample(UVs=Panner(Coordinate=Add(A=@n7, B=@n8), Time=@n5.Time) {speedY=-0.3}) {texture=/Game/Global/Textures/Default/Tileable_Patterns/T_Blowing_Smoke_Pattern2_Tile_DISP}
  n10 = LinearInterpolate(A=@n6.R, B=@n9.R)
  MaterialFunctionCall(Emissive=Multiply(A=@n1.RGB, B=LinearInterpolate(A=@n2, Alpha=@n10))) {materialFunction=&MF_FX_EmissiveAdjustments}
  n11 = Multiply(
   A=MaterialFunctionCall() {materialFunction=&VectorToRadialValue}.Radial Coordinates
   B=Constant2Vector() {g=0.3, r=3}
  )
  n12 = TextureSample(UVs=Panner(Coordinate=Add(A=@n8, B=@n11), Time=@n5.Time) {speedY=-0.1}) {texture=&T_Inky_Smoke_Tile}
  n13 = SphereMask(A=TextureCoordinate(), B=Constant() {r=0.5}) {attenuationRadius=0.5, hardnessPercent=0}
  n14 = Power(Base=@n13) {constExponent=0.5}
  n15 = Multiply(A=@n1.A, B=Power(Base=Multiply(A=@n9.R, B=@n12.R), Exp=LinearInterpolate(Alpha=@n14) {constA=3, constB=0}))
  n16 = Saturate(None=Multiply(A=Multiply(A=@n15)) {constB=3})
  n17 = Power(Base=@n13)
  n18 = OneMinus(None=SphereMask(A=TextureCoordinate(), B=Constant() {r=0.5}) {attenuationRadius=0.5, hardnessPercent=0})
  n19 = Saturate(None=LinearInterpolate(Alpha=Power(Base=@n18) {constExponent=1}) {constA=-0.1})
  n20 = ScalarParameter() {parameterName=UseCenterMask}
  Multiply(
   A=Multiply(A=Multiply(A=@n16, B=@n17), B=LinearInterpolate(B=@n19, Alpha=@n20) {constA=1})
   B=MaterialFunctionCall("FadeSensitivity (optional)"=ScalarParameter() {parameterName="FadeSensitivity (optional) (S)"}) {materialFunction=&FadeBasedOnViewAngle}.Result
  )
 ]
}
