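# Translucent, unlit, two-sided Niagara ribbon material for the lightning bolts of the LodeSwan laser VFX.
# Three offset samples of a four-row lightning flipbook, on panning, noise-distorted UVs, are masked by curve-atlas pulses scanning along the ribbon at separate speeds.
# Their refined sum feeds boosted color emissive, with optional particle tint, and opacity.
# A dynamic Grow parameter reveals length, while view-angle, camera-distance and depth fades soften opacity.
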Material /SpireSharedAssets/VFX/Laser/M_LodeSwan_Laser_LightningBolts
graph {
 properties {
  bAutomaticallySetUsageInEditor=false
  bUsedWithNiagaraRibbons=true
  blendMode=BLEND_Translucent
  shadingModel=MSM_Unlit
  twoSided=true
  usedShadingModels=Unlit
 }
 nodes [
  n1 = Subtract(A=PixelDepth().R, B=ScalarParameter() {defaultValue=75, parameterName=Clip Plane Distance})
  n2 = ScalarParameter() {defaultValue=500, parameterName=Soften}
  n3 = MaterialFunctionCall(XY=MaterialFunctionCall(
    "Animation  Phase (0-1)"=ScalarParameter() {parameterName=FlipbookPicker}
    Number of Rows=Constant() {r=4}
    Number of Columns=Constant() {r=1}
    UVs=NamedRerouteUsage()
   ) {materialFunction=&FlipBook}.UVs) {materialFunction=&Swizzle}
  n4 = TextureCoordinate() {coordinateIndex=1}
  n5 = Time()
  n6 = ScalarParameter() {defaultValue=1, parameterName=ScaleScanSpeedC}
  n7 = Frac(None=Add(A=Multiply(A=@n5, B=@n6), B=ScalarParameter() {parameterName=TimeOffset}))
  n8 = CurveAtlasRowParameter(CurveTime=Add(
    A=MaterialFunctionCall(UV Channel=@n4) {materialFunction=&LinearGradient}.UGradient
    B=LinearInterpolate(Alpha=@n7) {constA=-1, constB=0.75}
   )) {
   atlas=/Game/Effects/Fort_Effects/GameplayPulginsAssets/Demon/CA_HighTower_Date
   curve=/Game/Effects/Fort_Effects/GameplayPulginsAssets/Demon/Curve_HighTower_Date_Ball
   parameterName=Param_1
  }
  n9 = Multiply(
   A=NamedRerouteUsage()
   B=Multiply(A=TextureSample(UVs=@n3.YX) {texture=/Game/Effects/Fort_Effects/Textures/Electric/T_Lightning_Sprites}.R, B=Multiply(A=@n8.A))
  )
  n10 = ScalarParameter() {defaultValue=0.2, parameterName=Param_4}
  n11 = Add(A=ScalarParameter() {parameterName=FlipbookPicker}, B=@n10) {constB=0.33}
  n12 = MaterialFunctionCall(XY=MaterialFunctionCall("Animation  Phase (0-1)"=Frac(None=@n11), Number of Rows=Constant() {r=4}, Number of Columns=Constant() {r=1}, UVs=NamedRerouteUsage()) {materialFunction=&FlipBook}.UVs) {materialFunction=&Swizzle}
  n13 = TextureCoordinate() {coordinateIndex=1}
  n14 = Time()
  n15 = ScalarParameter() {defaultValue=2, parameterName=ScaleScanSpeedB}
  n16 = ScalarParameter() {parameterName=TimeOffset}
  n17 = Frac(None=Add(A=Multiply(A=@n14, B=@n15), B=Add(A=@n16, B=@n10) {constB=0.33}))
  n18 = CurveAtlasRowParameter(CurveTime=Add(
    A=MaterialFunctionCall(UV Channel=@n13) {materialFunction=&LinearGradient}.UGradient
    B=LinearInterpolate(Alpha=@n17) {constA=-1, constB=0.75}
   )) {
   atlas=&CA_HighTower_Date
   curve=&Curve_HighTower_Date_Ball
   parameterName=Param_1
  }
  n19 = Multiply(
   A=NamedRerouteUsage()
   B=Multiply(A=TextureSample(UVs=@n12.YX) {texture=&T_Lightning_Sprites}.R, B=Multiply(A=@n18.A))
  )
  n20 = NamedRerouteUsage()
  n21 = ScalarParameter() {parameterName=FlipbookPicker}
  n22 = Constant() {r=0.75}
  n23 = MaterialFunctionCall(
   "Animation  Phase (0-1)"=Frac(None=Add(A=@n21, B=@n22) {constB=0.33})
   Number of Rows=Constant() {r=4}
   Number of Columns=Constant() {r=1}
   UVs=NamedRerouteUsage()
  ) {materialFunction=&FlipBook}
  n24 = TextureCoordinate() {coordinateIndex=1}
  n25 = Time()
  n26 = ScalarParameter() {defaultValue=1.5, parameterName=ScaleScanSpeedA}
  n27 = ScalarParameter() {parameterName=TimeOffset}
  n28 = Frac(None=Add(A=Multiply(A=@n25, B=@n26), B=Add(A=@n27, B=@n22) {constB=0.33}))
  n29 = CurveAtlasRowParameter(CurveTime=Add(
    A=MaterialFunctionCall(UV Channel=@n24) {materialFunction=&LinearGradient}.UGradient
    B=LinearInterpolate(Alpha=@n28) {constA=-1, constB=0.75}
   )) {
   atlas=&CA_HighTower_Date
   curve=&Curve_HighTower_Date_Ball
   parameterName=Param_1
  }
  n30 = Multiply(
   A=TextureSample(UVs=MaterialFunctionCall(XY=@n23.UVs) {materialFunction=&Swizzle}.YX) {texture=&T_Lightning_Sprites}.R
   B=Multiply(A=@n29.A)
  )
  n31 = VectorParameter() {parameterName=L_Refine, defaultValue=<rgba 5 50 1 0.5>}
  n32 = MaterialFunctionCall(Texture=Add(A=Add(A=@n9, B=@n19), B=Multiply(A=@n20, B=@n30)), RGB=@n31.RGB, A=@n31.RGB) {materialFunction=&MF_FX_Refine}
  n33 = Subtract(
   A=MaterialFunctionCall(UV Channel=TextureCoordinate() {coordinateIndex=1}) {materialFunction=&LinearGradient}.UGradient
   B=OneMinus(None=DynamicParameter() {paramNames [Erode,Param2,IsUltra,Grow]}.Grow)
  )
  n34 = Multiply(A=@n32.Result, B=Saturate(None=Multiply(A=@n33) {constB=5}))
  n35 = Constant() {r=1.5}
  n36 = Saturate(None=Multiply(
    A=MaterialFunctionCall("FadeSensitivity (optional)"=@n35) {materialFunction=&FadeBasedOnViewAngle}.Result
    B=Constant() {r=20}
   ) {constB=10})
  n37 = Multiply(
   A=Saturate(None=Divide(A=@n1, B=@n2))
   B=DepthFade(Opacity=Multiply(A=@n34, B=@n36), FadeDistance=ScalarParameter() {defaultValue=150, parameterName=DepthFadeDis})
  )
  n38 = ParticleColor()
  LinearInterpolate(A=@n37, B=Multiply(A=@n38.A, B=@n37), Alpha=ScalarParameter() {parameterName=UsePAlpha})
  MaterialFunctionCall() {materialFunction=&DebugTimeSine}
  n39 = Multiply(A=VectorParameter() {parameterName=Color, defaultValue=<rgba 1 1 1 1>}.RGB, B=Constant() {r=4})
  n40 = Add(A=@n39, B=Multiply(A=Multiply(A=@n39, B=@n32.Result), B=Constant() {r=8}))
  MaterialFunctionCall(Emissive=LinearInterpolate(A=@n40, B=Multiply(A=@n38.RGB, B=@n40), Alpha=ScalarParameter() {parameterName=UserParticleColor})) {materialFunction=&MF_FX_EmissiveAdjustments}
  MaterialFunctionCall() {materialFunction=&DebugTimeSine}
  MaterialFunctionCall() {materialFunction=&DebugTimeSine}
  NamedRerouteDeclaration(None=MaterialFunctionCall(X=ScalarParameter() {defaultValue=0.1, parameterName=PanSpeedU}, Y=Constant()) {materialFunction=&MakeFloat2}.Result) {name=PanSpeed, nodeColor=<rgba 0 0.6953 1 1>}
  n41 = Multiply(
   A=TextureCoordinate()
   B=MaterialFunctionCall(X=Constant() {r=1}, Y=Constant() {r=1}) {materialFunction=&MakeFloat2}.Result
  )
  n42 = NamedRerouteUsage()
  n43 = TextureCoordinate()
  n44 = Constant2Vector() {g=1, r=2}
  n45 = TextureSample(UVs=Panner(Coordinate=Multiply(A=@n43, B=@n44)) {speedX=0.5, speedY=0.2}) {samplerType=SAMPLERTYPE_Normal, texture=/Game/Characters/Player/Female/Medium/Bodies/F_MED_Bistro_Astronaut/FX/T_GaussianNoise_Normal}
  n46 = MaterialFunctionCall(XY=MaterialFunctionCall(X=@n45.R, Y=@n45.R) {materialFunction=&MakeFloat2}.Result) {materialFunction=&Swizzle}
  n47 = ScalarParameter() {defaultValue=15, parameterName=Distortion}
  NamedRerouteDeclaration(None=Add(A=Panner(Coordinate=@n41, Speed=@n42), B=Divide(A=@n46.YX, B=@n47) {constB=10})) {name=UV, nodeColor=<rgba 1 0.0469 0 1>}
  n48 = CurveAtlasRowParameter(CurveTime=MaterialFunctionCall(UV Channel=TextureCoordinate() {coordinateIndex=1}) {materialFunction=&LinearGradient}.UGradient) {
   atlas=&CA_HighTower_Date
   curve=&Curve_HighTower_Date_Ball
   parameterName=Param_1
  }
  NamedRerouteDeclaration(None=Multiply(A=@n48.B)) {name=StartAndEndpointMask, nodeColor=<rgba 0.8047 0 1 1>}
 ]
}
