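# Unlit translucent Niagara particle material for a laser VFX, by its name a target center glow, usable on sprites, ribbons and meshes.
# An edge mask along V blends two colors, tinted by particle color for emissive.
# A panning plasma texture is refined separately for vertex-normal displacement and opacity.
# Opacity also combines inverted Fresnel, depth fade and particle alpha, softening silhouettes and intersections.
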Material /SpireSharedAssets/VFX/Laser/M_TargetCenter2
graph {
 properties {
  bAutomaticallySetUsageInEditor=false
  bUsedWithNiagaraMeshParticles=true
  bUsedWithNiagaraRibbons=true
  bUsedWithNiagaraSprites=true
  blendMode=BLEND_Translucent
  shadingModel=MSM_Unlit
  usedShadingModels=Unlit
 }
 nodes [
  n1 = TextureCoordinate()
  n2 = VectorParameter() {
   group=Color
   parameterName=Color_EdgeMask_InvertControl
   channelNames {a="", b=Power, g=Multiply, r=Offset}
   defaultValue=<rgba -0.0768 5 2 1>
  }
  n3 = VectorParameter() {
   group=Color
   parameterName=Color_EdgeMask_Control
   channelNames {a="", b=Power, g=Multiply, r=Offset}
   defaultValue=<rgba 0.3136 1.7816 0.348 1>
  }
  n4 = LinearInterpolate(
   A=VectorParameter() {parameterName=Color, defaultValue=<rgba 100 10 17.4999 1>}.RGB
   B=VectorParameter() {parameterName=ColorD, defaultValue=<rgba 1.5 0 0.125 1>}.RGB
   Alpha=MaterialFunctionCall!(
    UorV=MaterialFunctionCall(Float2=@n1) {materialFunction=&BreakOutFloat2Components}.G
    Invert_EdgeControl=@n2
    NoInvert_EdgeControl=@n3
   ) {materialFunction=/Game/Effects/Niagara/MasterMaterials/Functions/Common/FXCommon_EdgeMask_Ex}.Mask
  )
  n5 = ParticleColor()
  MaterialFunctionCall(Emissive=Divide(A=Multiply(A=@n4, B=@n5.RGB)) {constB=3.5}) {materialFunction=&MF_FX_EmissiveAdjustments}
  n6 = TextureObjectParameter() {
   group=Texture
   parameterName=Texture_Texture
   samplerType=SAMPLERTYPE_LinearGrayscale
   sortPriority=0
   texture=/Game/Effects/Fort_Effects/Emotes/UltraLight/Texture/T_Plasma_A_B_Greyscale
  }
  n7 = VectorParameter() {group=Texture, parameterName=Texture_UVControl, channelNames {a=VPan, b=UPan, g=VScale, r=UScale}, defaultValue=<rgba 1 0.5 0 -1.5>}
  n8 = MaterialFunctionCall(Texture=@n6, UVControl_RGB=@n7, UVControl_A=@n7) {materialFunction=&FXCommon_TextureSample}
  n9 = VectorParameter() {parameterName=Texture_RefineTOWPO, defaultValue=<rgba 1.5224 -22.6084 0 0.5>}
  Multiply(
   A=MaterialFunctionCall(Texture=@n8.R, RGB=@n9.RGB, A=@n9.RGB) {materialFunction=&MF_FX_Refine}.Result
   B=VertexNormalWS()
  )
  n10 = VectorParameter() {parameterName=Texture_Refine, defaultValue=<rgba 1.5 5 1 0.5>}
  n11 = ScalarParameter() {defaultValue=4.9025, parameterName=EXPO}
  n12 = Power(Base=OneMinus(None=Fresnel(ExponentIn=@n11) {baseReflectFraction=0}), Exp=ScalarParameter() {defaultValue=14.2896, parameterName=EXP2})
  n13 = Power(
   Base=DepthFade(FadeDistance=ScalarParameter() {defaultValue=50, parameterName=DF})
   Exp=ScalarParameter() {defaultValue=4.4714, parameterName=DFP}
  )
  Multiply(
   A=@n5.A
   B=Multiply(
    A=MaterialFunctionCall(Texture=@n8.R, RGB=@n10.RGB, A=@n10.RGB) {materialFunction=&MF_FX_Refine}.Result
    B=Multiply(A=@n12, B=@n13)
   )
  )
 ]
}
