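# Niagara system for the charge-up of the large Spire laser, timed largely by normalized system loop age.
# Airflow spawns growing meshes at a rising rate, collapses them late and dims them by time of day.
# Attractor pulls GPU sprites from a flattened sphere toward the core, shifting spawn color from secondary to primary.
# PP drives a component renderer with a distance-faded intensity curve and a Duration-based enable check; the other emitters list no renderers.
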NiagaraSystem /SpireSharedAssets/VFX/Laser/NS_LargeSpire_Charge
dependencies {
 usedDataInterfaces [&NiagaraDataInterfaceCurve, &NiagaraDataInterfaceColorCurve]
 usedDynamicInputs [
  &SimulationPosition
  &Waveform
  &MakeVector4
  &MakeLinearColorFromVectorAndFloat
  &ClampFloat
  &Divide_Float
  &Subtract_Float
  &MakeFloatFromVector
  /Game/Effects/Niagara/DynamicInputs/Math/FN_Fit
 ]
 usedModules [
  &self:SystemSpawnScript.SystemScriptSource.SystemScriptGraph.NiagaraNodeAssignment_0.SetVariables_#1
  &self:SystemSpawnScript.SystemScriptSource.SystemScriptGraph.NiagaraNodeAssignment_1.SetVariables_#2
  &self:SystemSpawnScript.SystemScriptSource.SystemScriptGraph.NiagaraNodeAssignment_2.SetVariables_#3
  &SystemState
  &self:SystemSpawnScript.SystemScriptSource.SystemScriptGraph.NiagaraNodeAssignment_3.SetVariables_#4
  &self:SystemSpawnScript.SystemScriptSource.SystemScriptGraph.NiagaraNodeAssignment_5.SetVariables_#5
  &self:SystemSpawnScript.SystemScriptSource.SystemScriptGraph.NiagaraNodeAssignment_4.SetVariables_#6
  &self:SystemSpawnScript.SystemScriptSource.SystemScriptGraph.NiagaraNodeAssignment_6.SetVariables_#7
  &self:SystemSpawnScript.SystemScriptSource.SystemScriptGraph.NiagaraNodeAssignment_7.SetVariables_#8
  &self:SystemSpawnScript.SystemScriptSource.SystemScriptGraph.NiagaraNodeAssignment_8.SetVariables_#9
 ]
}
emitters {
 Airflow {
  emitterUpdateScript [
   &EmitterState()
   &SpawnBurst_Instantaneous(Spawn Count=5, Use Spawn Probability=false, UseLoopCountLimit=false) {enabled=false}
   &SpawnRate(
    SpawnRate=&FloatFromCurve(
     FloatCurve=DataInterface() {propertyValues {Curve {keys [{interpMode=RCIM_Cubic, leaveTangent=3.6882, tangentMode=RCTM_Break, time=0, value=0}, <at 1 1>]}}}
     CurveIndex=$System.NormalizedLoopAge
     Scale Curve=20
    )
    Use Spawn Probability=true
   ) {module=SpawnRate001}
  ]
  particleSpawnScript [
   &InitializeParticle(
    Color=<rgba 0.025 0 0.0021 8>
    Lifetime=&RandomRangeFloat(Minimum=0.9, Maximum=0.9)
    Mass Mode=::NewEnumerator1
    Mesh Scale=&Multiply_VectorByFloat(Vector=&RandomRangeVector(Minimum=<0.9 0.9 0.9>, Maximum=<1 1 1>), Float=3)
    Mesh Scale Mode=::NewEnumerator4
    Position Offset=&AddVectorToPosition(Position=$System.FXCenter, Vector=<0 0 0>)
    Position Offset Coordinate Space=::Local
    Sprite Size=<5 5>
    Sprite Size Mode=::NewEnumerator4
    UsePositionOffset=true
    Write Material Random=true
   ) {module=InitializeParticle001}
   &InitialMeshOrientation(Use Rotation Vector=false)
   &CameraOffset(Camera Offset Amount=100) {module=CameraOffset001}
  ]
  particleUpdateScript [
   &ParticleState()
   &ScaleMeshSize(Scale Factor=&VectorFromFloat(Value=&FloatFromCurve(
      FloatCurve=DataInterface() {propertyValues {Curve {keys [
          {interpMode=RCIM_Cubic, tangentMode=RCTM_Break, time=0.5, value=1}
          {arriveTangent=-7.6859, interpMode=RCIM_Cubic, tangentMode=RCTM_Break, time=0.9, value=0}
         ]}}}
      CurveIndex=$System.NormalizedLoopAge
     ))) {module=ScaleMeshSize001}
   &SolveForcesAndVelocity(Clamp Velocity=false, Limit Acceleration=false)
   &DynamicMaterialParameters(
    Index 0 Param 1=&FloatFromCurve(FloatCurve=DataInterface() {propertyValues {Curve {keys [<at 0 -0.2>,<at 1 1.2>]}}})
    Index 0 Param 2=1
    Index 0 Param 3=&Multiply_Float(
     A=&RandomRangeFloat(Minimum=0.8, Maximum=1)
     B=&FloatFromCurve(FloatCurve=DataInterface() {propertyValues {Curve {keys [<at 0 1>,<at 1 0.5>]}}}, Scale Curve=20)
    )
    Module.Param0WriteEnabled=true
    Module.Param1WriteEnabled=true
    Module.Param2WriteEnabled=true
    Module.Param3WriteEnabled=true
   )
   &ScaleColor(
    Scale Alpha=&Multiply_Float(
     A=&Multiply_Float(
      A=&Multiply_Float(A=1, B=&FNTimeOfDayScalar(Night Scale=1, Morning Scale=0.3, Evening Scale=0.4))
      B=&FloatFromCurve(FloatCurve=DataInterface() {propertyValues {Curve {keys [
           <at 0 0 RCIM_Cubic>
           <at 0.1912 0.9932>
           {arriveTangent=-1.2279, interpMode=RCIM_Cubic, leaveTangent=-1.2279, time=0.7879, value=1.0115}
           <at 1 0>
          ]}}})
     )
     B=&FloatFromCurve(
      FloatCurve=DataInterface() {propertyValues {Curve {keys [
          {interpMode=RCIM_Cubic, tangentMode=RCTM_User, time=0, value=0}
          {arriveTangent=0.37, interpMode=RCIM_Cubic, leaveTangent=0.37, tangentMode=RCTM_User, time=0.6, value=0.05}
          {arriveTangent=1.1869, interpMode=RCIM_Cubic, leaveTangent=1.1869, tangentMode=RCTM_User, time=0.9, value=0.5258}
          <at 1.2 0.8>
         ]}}}
      CurveIndex=$Emitter.Age
     )
    )
    ScaleA=true
    ScaleRGB=true
   ) {enabled=false}
   &ScaleColor(
    Scale Alpha=&Multiply_Float(
     A=&Multiply_Float(
      A=&Multiply_Float(A=1, B=&FNTimeOfDayScalar(Night Scale=1, Morning Scale=0.3, Evening Scale=0.4))
      B=&FloatFromCurve(FloatCurve=DataInterface() {propertyValues {Curve {keys [<at 0.1912 0.9932>, {interpMode=RCIM_Cubic, leaveTangent=-1.2279, time=0.7879, value=1.0115}]}}})
     )
     B=$System.NormalizedLoopAge
    )
    ScaleA=true
    ScaleRGB=true
   ) {module=ScaleColor002}
   &ScaleColor(
    Scale Alpha=&Multiply_Float(
     A=0.5
     B=&FloatFromCurve(FloatCurve=DataInterface() {propertyValues {Curve {keys [
          {arriveTangent=9.0725, interpMode=RCIM_Cubic, leaveTangent=9.0725, tangentMode=RCTM_User, time=0, value=0}
          {arriveTangent=-2.5249, interpMode=RCIM_Cubic, leaveTangent=-2.5249, tangentMode=RCTM_User, time=0.1408, value=0.9847}
          {arriveTangent=-0.0405, interpMode=RCIM_Cubic, leaveTangent=-0.0405, tangentMode=RCTM_User, time=1, value=0}
         ]}}})
    )
    Scale RGB=&Multiply_Vector(
     A=&MakeVectorFromLinearColorRGB(Color=&ColorFromCurve(ColorCurve=DataInterface() {propertyValues {
         AlphaCurve {keys [<at 0 1 RCIM_Cubic>,<at 0.9977 1 RCIM_Cubic>]}
         BlueCurve {keys [
           <at 0.0057 4.809>
           {arriveTangent=-9.9261, interpMode=RCIM_Cubic, leaveTangent=-9.9261, time=0.1454, value=1.0084}
           <at 0.4346 0.5512>
           <at 0.5279 2.0319>
           <at 0.7175 2.7561>
           {arriveTangent=0.9451, interpMode=RCIM_Cubic, leaveTangent=0.9451, time=0.8517, value=1.9922}
           <at 0.9448 2.9709>
          ]}
         GreenCurve {keys [
           <at 0.0057 4.7917>
           {arriveTangent=-9.9807, interpMode=RCIM_Cubic, leaveTangent=-9.9807, time=0.1454, value=0.9183}
           <at 0.4346 0.5104>
           <at 0.5279 1.762>
           <at 0.7175 2.5521>
           {arriveTangent=1.0311, interpMode=RCIM_Cubic, leaveTangent=1.0311, time=0.8517, value=1.7188}
           <at 0.9448 2.7865>
          ]}
         RedCurve {keys [
           <at 0.0057 5>
           {arriveTangent=-9.3251, interpMode=RCIM_Cubic, leaveTangent=-9.3251, time=0.1454, value=2}
           <at 0.4346 1>
           <at 0.5279 5>
           <at 0.7175 5>
           <at 0.8517 5 RCIM_Cubic>
           <at 0.9448 5>
          ]}
        }}))
     B=&VectorFromFloat(Value=2)
    )
    ScaleA=false
    ScaleRGB=true
   ) {module=ScaleColor001}
   &ScaleMeshSize(Scale Factor=&Multiply_Vector(
     A=<1 1 1>
     B=&VectorFromFloat(Value=&FloatFromCurve(
       FloatCurve=DataInterface() {propertyValues {Curve {keys [
           {arriveTangent=0.2635, interpMode=RCIM_Cubic, leaveTangent=0.2635, tangentMode=RCTM_User, time=0, value=0.2}
           {arriveTangent=0.4932, interpMode=RCIM_Cubic, leaveTangent=0.4932, tangentMode=RCTM_User, time=0.5, value=0.5}
           {arriveTangent=0.447, interpMode=RCIM_Cubic, leaveTangent=0.447, tangentMode=RCTM_User, time=1, value=1.2}
          ]}}}
       CurveIndex=$Emitter.Age
       Scale Curve=2
      ))
    ))
   &CameraOffset(Camera Offset Amount=20)
  ]
  rendererClasses [&NiagaraMeshRendererProperties]
  renderers [{rendererClass=&NiagaraMeshRendererProperties, rendererIndex=0}]
 }
 Attractor {
  simTarget=GPUComputeSim
  emitterUpdateScript [
   &EmitterState()
   &SpawnRate(
    Spawn Probability=&FloatFromCurve(
     FloatCurve=DataInterface() {propertyValues {Curve {keys [
         {interpMode=RCIM_Cubic, leaveTangent=-0.5379, tangentMode=RCTM_Break, time=0, value=0}
         {interpMode=RCIM_Cubic, tangentMode=RCTM_Break, time=0.7, value=1}
         {arriveTangent=-8.2237, tangentMode=RCTM_User, time=1, value=0}
        ]}}}
     CurveIndex=$System.NormalizedLoopAge
    )
    SpawnRate=150
    Use Spawn Probability=true
   )
  ]
  particleSpawnScript [
   &InitializeParticle(
    Color=&LerpLinearColors(StartColor=$User.SecondaryColor, EndColor=$User.PrimaryColor, LerpFactor=$System.NormalizedLoopAge)
    Lifetime=0.4
    Position Offset=&AddVectorToPosition(Position=Linked() {linkedVariable {name=System.FXCenter, type=Vector3f}}, Vector=<0 0 0>)
    Position Offset Coordinate Space=::Local
    Sprite Size Max=<2 7>
    Sprite Size Min=<1 5>
    Sprite Size Mode=::NewEnumerator2
    UsePositionOffset=true
    Write Material Random=true
   )
   &ShapeLocation(
    MaskSpawnGroup=false
    Non Uniform Scale=<1 1 0.3>
    Override Seed=false
    Rotation Mode=::NewEnumerator5
    Sphere Radius=&Multiply_Float(A=883.3219, B=1)
    Sphere Surface Distribution=1
    UseOwnerScale=true
   )
  ]
  particleUpdateScript [
   &ParticleState()
   &ScaleSpriteSize(Scale Sprite Size Mode=::NewEnumerator0, Uniform Scale Factor=3)
   &PointAttractionForce(
    Attraction Radius=4000
    Attraction Strength=&Multiply_Float(A=318.0403, B=2)
    Attractor Position Offset:NiagaraPosition=$System.CorePosition
    Kill Within Radius=false
    Module.Use Falloff=true
    UseAttractorOffset=false
   )
   &CurlNoiseForce(Bias Noise=false, Noise Frequency=10, Noise Strength=0, Pan Noise=false)
   &Drag(
    Drag=&FloatFromCurve(FloatCurve=DataInterface() {propertyValues {Curve {keys [<at 0 1>,<at 1 0>]}}}, Scale Curve=5)
    Use Linear Drag=true
    Use Rotational Drag=true
   )
   &SolveForcesAndVelocity(Clamp Velocity=false, Limit Acceleration=false)
   &ScaleSpriteSizeBySpeed(Max Scale Factor=<1 3>, Min Scale Factor=<0.5 0.5>)
   &FN_Sprite_Attributes(
    Color=&MultiplyLinearColor(Linear Color=$Particles.Color, Scale Factor=&MakeVector4_FromLinearColor(LinearColor=<rgba 1 1 1 0.02>))
    Create Color Attribute=true
    Create SpriteSize Attribute=true
    SpriteSize=&Multiply_Vector2D(A=$Particles.SpriteSize, B=<20 3>)
    Write to Module Namespace=true
   )
   &DynamicMaterialParameters(
    Index 0 Param 1=2
    Module.Index0Param0Write=true
    Module.Index0Param1Write=true
    Module.Index0Param2Write=true
    Module.Index0Param3Write=true
   )
  ]
  rendererClasses [&NiagaraSpriteRendererProperties]
  renderers [
   {rendererClass=&NiagaraSpriteRendererProperties}
   {rendererClass=&NiagaraSpriteRendererProperties, rendererIndex=1}
  ]
 }
 BigGlow {rendererClasses [], renderers []}
 BrightCenter {rendererClasses [], renderers []}
 CoreFlare {rendererClasses [], renderers []}
 Corona003 {rendererClasses [], renderers []}
 Corona004 {rendererClasses [], renderers []}
 Corona005 {rendererClasses [], renderers []}
 Corona006 {rendererClasses [], renderers []}
 DarkCenter {rendererClasses [], renderers []}
 PP {
  emitterUpdateScript [
   &EmitterState()
   &SpawnBurst_Instantaneous(Use Spawn Probability=true, UseLoopCountLimit=true)
   &SpawnBurst_Instantaneous(Spawn Time=0.3167, Use Spawn Probability=false, UseLoopCountLimit=false) {module=SpawnBurst_Instantaneous001}
  ]
  particleSpawnScript [&InitializeParticle(
    Color Mode=::NewEnumerator0
    Lifetime=&Add_Float(A=$System.CurrentLoopDuration, B=0)
    Sprite Size Mode=::NewEnumerator3
    Uniform Sprite Size=200
    UsePositionOffset=false
    Write Material Random=true
   )]
  particleUpdateScript [
   &ParticleState()
   /SetVariables(Particles.SpeedNormalizedB:NiagaraFloat=&FloatFromCurve(
     FloatCurve=DataInterface() {propertyValues {Curve {keys [{arriveTangent=2.8002, interpMode=RCIM_Cubic, leaveTangent=2.8002, tangentMode=RCTM_User, time=-0.001, value=0}, <at 0.75 0.95 RCIM_Cubic>]}}}
     Scale Curve=&Multiply_Float(
      A=0.8
      B=&FN_LOD_Distance_Remap(Start Distance=4000, End Distance=7000, Minimum Output=1, Maximum Output=0, Clamp Output=true)
     )
    )) {isSetParameters=true}
   /SetVariables(Particles.ComponentsEnabled:NiagaraBool=&SetBoolByFloatComparison(A=$Engine.System.Age, B=&Add_Float(A=$User.Duration, B=0.5), Comparison Type=::NewEnumerator6)) {isSetParameters=true}
  ]
  rendererClasses [&NiagaraComponentRendererProperties]
  renderers [{rendererClass=&NiagaraComponentRendererProperties, rendererIndex=0}]
 }
 ShockWave {rendererClasses [], renderers []}
 ShockWave001 {rendererClasses [], renderers []}
 WaveRemain001 {rendererClasses [], renderers []}
}
summary {
 systemName=NS_LargeSpire_Charge
 emitters [
  {emitterName=Airflow, rendererClasses [&NiagaraMeshRendererProperties]}
  {emitterName=Attractor, simTarget=GPUComputeSim, rendererClasses [&NiagaraSpriteRendererProperties]}
  {emitterName=PP, rendererClasses [&NiagaraComponentRendererProperties]}
  {emitterName=Corona003, rendererClasses []}
  {emitterName=Corona004, rendererClasses []}
  {emitterName=Corona005, rendererClasses []}
  {emitterName=WaveRemain001, rendererClasses []}
  {emitterName=CoreFlare, rendererClasses []}
  {emitterName=BigGlow, rendererClasses []}
  {emitterName=DarkCenter, rendererClasses []}
  {emitterName=BrightCenter, rendererClasses []}
  {emitterName=Corona006, rendererClasses []}
  {emitterName=ShockWave, rendererClasses []}
  {emitterName=ShockWave001, rendererClasses []}
 ]
 userVariables [
  User.bIsFrontEnd {type=NiagaraBool}
  User.bIsFrontEndPreview {type=NiagaraBool}
  User.Duration {type=NiagaraFloat, defaultValue=2.4, description=""}
  User.ElementSizeScale {type=NiagaraFloat, defaultValue=1, description=""}
  User.isBigPOV {type=NiagaraBool, defaultValue=true}
  User.IsLocalPlayer {type=NiagaraBool, defaultValue=true}
  User.PrimaryColor {type=LinearColor, description="", defaultValue {struct=LinearColor, value=<rgba 5 0.2882 0.6068 1>}}
  User.SecondaryColor {type=LinearColor, description="", defaultValue {struct=LinearColor, value=<rgba 0.7101 0.0959 0.1594 5>}}
  User.Stop Location {type=NiagaraPosition, description="", defaultValue {}}
 ]
}
